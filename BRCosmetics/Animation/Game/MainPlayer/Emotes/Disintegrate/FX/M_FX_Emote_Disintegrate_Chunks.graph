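# Default Lit material for the chunk mesh particles in the Disintegrate emote FX, enabled for Niagara mesh particles.
# It samples diffuse, normal and specular-mask textures from the Disintegrate gadget prop, all exposed as parameters.
# The mask's blue channel blends between min and max roughness parameters, and its green channel is scaled by a Metallic parameter.
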Material /BRCosmetics/Animation/Game/MainPlayer/Emotes/Disintegrate/FX/M_FX_Emote_Disintegrate_Chunks
graph {
 properties {bUsedWithNiagaraMeshParticles=true, pixelDepthOffsetMode=PDOM_Legacy, usedShadingModels=Default Lit}
 nodes [
  n1 = TextureSampleParameter2D() {
   group=Specular
   parameterName=SpecularMasks
   samplerType=SAMPLERTYPE_Masks
   texture=/BRCosmetics/Animation/Props/Gadgets_Disintegrate/Textures/T_Emote_Prop_Disintegrate_S
  }
  LinearInterpolate(
   A=ScalarParameter() {defaultValue=0.22, group=Specular, parameterName=SpecRoughnessMin, sliderMax=1}
   B=ScalarParameter() {defaultValue=0.66, group=Specular, parameterName=SpecRoughnessMax, sliderMax=1}
   Alpha=@n1.B
  )
  Multiply(A=@n1.G, B=ScalarParameter() {defaultValue=1, group=Specular, parameterName=Metallic, sliderMax=1})
  TextureSampleParameter2D() {
   parameterName=Normals
   samplerType=SAMPLERTYPE_Normal
   texture=/BRCosmetics/Animation/Props/Gadgets_Disintegrate/Textures/T_Emote_Prop_Disintegrate_N
  }
  TextureSampleParameter2D() {parameterName=Diffuse, texture=/BRCosmetics/Animation/Props/Gadgets_Disintegrate/Textures/T_Emote_Prop_Disintegrate_D}
 ]
}
