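# Translucent unlit Niagara sprite material for impact spikes in the Disintegrate emote effects.
# Panning noise distorts UVs of an explosion streak texture, which combines with particle alpha, a tiling mask and an optional noise-distorted dissolve.
# Dynamic parameters drive noise strength, dissolve amount and random time offsets.
# Opacity takes optional depth and near-camera fades, while emissive uses particle color, optionally tinted by time of day.
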Material /BRCosmetics/Animation/Game/MainPlayer/Emotes/Disintegrate/FX/M_ImpactSpikes
graph {
 properties {
  bAllowNegativeEmissiveColor=true
  bUsedWithNiagaraSprites=true
  blendMode=BLEND_Translucent
  pixelDepthOffsetMode=PDOM_Legacy
  shadingModel=MSM_Unlit
  usedShadingModels=Unlit
 }
 nodes [
  n1 = ParticleColor()
  n2 = LinearInterpolate(
   A=TextureCoordinate()
   B=TextureCoordinate() {coordinateIndex=1}
   Alpha=ScalarParameter() {group="Noise ", parameterName="Use Noise TexCoord [0] or [1]"}
  )
  n3 = VectorParameter() {group="Noise ", parameterName=Noise UV, defaultValue=<rgba 1 1 0 0>}
  n4 = MaterialFunctionCall(X=@n3.R, Y=@n3.R) {materialFunction=&MakeFloat2}
  n5 = MaterialFunctionCall(X=@n3.B, Y=@n3.B) {materialFunction=&MakeFloat2}
  n6 = DynamicParameter() {parameterIndex=3, defaultValue=<rgba 0 0 1 1>, paramNames [UVTilingRandom, TimeOffsetRandom, SplineWidthBase, SplineWidthTip]}
  n7 = Add(A=Multiply(A=Time(), B=ScalarParameter() {defaultValue=1, parameterName=NoisePanningSpeed}), B=@n6.TimeOffsetRandom)
  n8 = TextureSample(
   UVs=Add(A=Multiply(A=@n2, B=@n4.Result), B=Multiply(A=@n5.Result, B=@n7))
   Tex=TextureObjectParameter() {group="Noise ", parameterName=Noise Tex}
  )
  n9 = ScalarParameter() {defaultValue=-1, group="Noise ", parameterName=Noise Strength}
  n10 = DynamicParameter() {defaultValue=<rgba 0 1 1 1>, paramNames [Emi V Offset, Emi V Tiling, Noise, Dissolve]}
  n11 = Add(
   A=Multiply(A=ComponentMask(None=@n8.RGB) {g=true, r=true}, B=Multiply(A=@n9, B=@n10.Noise))
   B=ScalarParameter() {group="Noise ", parameterName=Noise Offset}
  )
  n12 = Add(A=TextureCoordinate(), B=@n11)
  n13 = VectorParameter() {defaultValue=<rgba 1 1 0 0>}
  n14 = Multiply(
   A=MaterialFunctionCall!(
    TexCoord=@n12
    ClampChannel=MaterialFunctionCall(X=@n13.R, Y=@n13.R) {materialFunction=&MakeFloat2}.Result
   ) {materialFunction=/Game/Effects/Fort_Effects/MaterialFunctions/MF_FX_TilingMethod}.Result
   B=TextureSample(UVs=@n12) {texture=/Game/Effects/Fort_Effects/Gameplay/SmokeDustExplosions/Textures/T_Explosion_Streak_02}.RGB
  )
  n15 = NamedRerouteUsage()
  n16 = VectorParameter() {
   group=Dissolve
   parameterName=Dissolve Control2
   sortPriority=2
   channelNames {a="", b=Tex Multiply, g=Use Position Dissolve, r=Noise Strength}
   defaultValue=<rgba 0 0 0.5 0>
  }
  n17 = VectorParameter() {
   group=Dissolve
   parameterName=Dissolve Control
   sortPriority=1
   channelNames {a=Intensity, b=Smooth, g=Use Pan of time, r="Use TexCoord [0] or [1]"}
   defaultValue=<rgba 0 0 3 1>
  }
  n18 = VectorParameter() {group=Dissolve, parameterName=Dissolve UV, sortPriority=3, defaultValue=<rgba 1 1 0 0>}
  n19 = Multiply(
   A=LinearInterpolate(A=TextureCoordinate(), B=TextureCoordinate() {coordinateIndex=1}, Alpha=@n17."Use TexCoord [0] or [1]")
   B=MaterialFunctionCall(X=@n18.R, Y=@n18.R) {materialFunction=&MakeFloat2}.Result
  )
  n20 = MaterialFunctionCall(X=@n18.B, Y=@n18.B) {materialFunction=&MakeFloat2}
  n21 = Time()
  n22 = DynamicParameter() {parameterIndex=3, defaultValue=<rgba 0 0 1 1>, paramNames [UVTilingRandom, TimeOffsetRandom, SplineWidthBase, SplineWidthTip]}
  n23 = LinearInterpolate(A=@n20.Result, B=Multiply(A=@n20.Result, B=Add(A=@n21, B=@n22.TimeOffsetRandom)), Alpha=@n17.Use Pan of time)
  n24 = TextureSample(
   UVs=Add(A=Multiply(A=@n11, B=@n16.Noise Strength), B=Add(A=@n19, B=@n23))
   Tex=TextureObjectParameter() {group=Dissolve, parameterName=Dissolve Tex, sortPriority=2}
  )
  n25 = Add(A=Multiply(A=@n24.R, B=@n16.Tex Multiply))
  n26 = Multiply(A=LinearInterpolate(A=@n24.R, B=Saturate(None=@n25), Alpha=@n16.Use Position Dissolve), B=@n17.Smooth)
  n27 = LinearInterpolate(
   A=@n17.Smooth
   Alpha=Multiply(A=DynamicParameter() {defaultValue=<rgba 0 1 1 1>, paramNames [Emi V Offset, Emi V Tiling, Noise, Dissolve]}.Dissolve, B=@n17.Intensity)
  ) {constB=-1}
  n28 = Multiply(A=@n15, B=Saturate(None=Subtract(A=@n26, B=@n27)))
  n29 = ScalarParameter() {group=Dissolve, parameterName=Use Dissolve, sortPriority=0}
  n30 = Multiply(A=Multiply(A=Multiply(A=@n1.A, B=@n14), B=LinearInterpolate(A=@n15, B=@n28, Alpha=@n29)))
  n31 = ScalarParameter() {defaultValue=5, group=Opacity|DepthFade, parameterName=DepthFade_Distance, sortPriority=2}
  n32 = LinearInterpolate(
   A=@n30
   B=ShadingPathSwitch(Default=DepthFade(Opacity=@n30, FadeDistance=@n31), Mobile=@n30)
   Alpha=ScalarParameter() {group=Opacity|DepthFade, parameterName=Use Depth Fade}
  )
  n33 = Power(
   Base=MaterialFunctionCall(Distance=ScalarParameter() {defaultValue=512, group=Opacity|Near_Camera_Fade, parameterName=NearCamFade_Distance, sortPriority=1}) {materialFunction=&Near_Camera_Fade}.Result
   Exp=ScalarParameter() {defaultValue=1, group=Opacity|Near_Camera_Fade, parameterName=NearCamFade_Power, sortPriority=2}
  )
  LinearInterpolate(
   A=@n32
   B=Multiply(A=@n32, B=Clamp(None=@n33) {minDefault=0.02})
   Alpha=ScalarParameter() {group=Opacity|Near_Camera_Fade, parameterName=Add Near Camera Fade}
  )
  n34 = MaterialFunctionCall(
   Color_Morning=VectorParameter() {parameterName=Color Morning, defaultValue=<rgba 0.2795 0.1398 0.1407 1>}.RGB
   Color_Day=VectorParameter() {parameterName=Color Day, defaultValue=<rgba 0.6094 0.545 0.5552 1>}.RGB
   Color_Evening=VectorParameter() {parameterName=Color Evening, defaultValue=<rgba 0.2726 0.1761 0.1282 1>}.RGB
   Color_Night=VectorParameter() {parameterName=Color Night, defaultValue=<rgba 0.3438 0.3438 0.3438 1>}.RGB
  ) {materialFunction=&Color_Time_of_Day}
  MaterialFunctionCall(Emissive=StaticSwitchParameter(True=Multiply(A=@n34.Result, B=@n1.RGB), False=@n1.RGB) {parameterName=UseTimeOfDayColor}) {materialFunction=&MF_FX_EmissiveAdjustments}
  n35 = VectorParameter() {group=Mask, parameterName=Mask UV, sortPriority=1, defaultValue=<rgba 1 1 0 0>}
  n36 = DynamicParameter() {parameterIndex=2, defaultValue=<rgba 0 0 1 0>, paramNames [Sine Speed, Sine Tile, Sine Strength, Param4]}
  n37 = LinearInterpolate(A=@n35.G, B=Multiply(A=@n35.G, B=@n36.Param4), Alpha=ScalarParameter() {parameterName=UseDPtoAnimateMaskVtile})
  n38 = MaterialFunctionCall(X=@n35.B, Y=@n35.B) {materialFunction=&MakeFloat2}
  n39 = DynamicParameter() {parameterIndex=3, defaultValue=<rgba 0 0 1 1>, paramNames [UVTilingRandom, TimeOffsetRandom, SplineWidthBase, SplineWidthTip]}
  n40 = Add(A=Time(), B=@n39.TimeOffsetRandom)
  n41 = VectorParameter() {
   group=Mask
   parameterName=Mask TilingMethod
   sortPriority=2
   channelNames {a="Use TexCoord [0] or [1]", b=Rotation, g=V Clamp, r=U Clamp}
   defaultValue=<rgba 1 1 0 0>
  }
  n42 = MaterialFunctionCall(X=@n41.U Clamp, Y=@n41.U Clamp) {materialFunction=&MakeFloat2}
  n43 = Add(
   A=Multiply(
    A=TextureCoordinate()
    B=MaterialFunctionCall(X=@n35.R, Y=@n37) {materialFunction=&MakeFloat2}.Result
   )
   B=LinearInterpolate(A=Multiply(A=@n38.Result, B=@n40), B=@n38.Result, Alpha=@n42.Result)
  )
  n44 = MaterialFunctionCall(TexCoord=@n43, ClampChannel=@n42.Result) {materialFunction=&MF_FX_TilingMethod}
  n45 = TextureSample(UVs=@n43, Tex=TextureObjectParameter() {group=Mask, parameterName=Mask Tex, sortPriority=0})
  NamedRerouteDeclaration(None=Multiply(A=Multiply(A=@n44.Result, B=@n45.R), B=ScalarParameter() {defaultValue=1, group=Mask, parameterName=Mask Multiply, sortPriority=3})) {name=Mask, nodeColor=<rgba 0 0.2734 1 1>}
 ]
}
