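# Niagara burst effect for a ground impact in the Disintegrate emote, throwing debris, smoke, sparks and streaks.
# Sprite emitters provide air streaks, floor spikes, a smoke ring and distance-culled sparks, with time-of-day alpha scaling on all but sparks.
# Two enabled mesh emitters, Wood and Guitar_Elements_Backup, launch two spawn groups of weighted-random chunks with delayed gravity, analytical-plane collision gated by quality level, and shrink-out.
# A debug gnomon emitter is disabled.
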NiagaraSystem /BRCosmetics/Animation/Game/MainPlayer/Emotes/Disintegrate/FX/NS_GroundImpact
dependencies {
 usedDataInterfaces [
  &NiagaraDataInterfaceCurve
  &NiagaraDataInterfaceMeshRendererInfo
  &NiagaraDataInterfaceVector2DCurve
  &NiagaraDataInterfaceColorCurve
  &NiagaraDataInterfaceArrayFloat3
  &NiagaraDataInterfacePlatformSet
 ]
 usedDynamicInputs [&HueShiftLinearColor]
 usedModules [
  &self:SystemSpawnScript.SystemScriptSource.SystemScriptGraph.NiagaraNodeAssignment_0.SetVariables_#1
  &SystemState
 ]
}
emitters {
 AirStreaks001 {
  emitterUpdateScript [
   &EmitterState()
   &SpawnBurst_Instantaneous(
    Spawn Count=&RandomRangeInt(Minimum=8, Maximum=11)
    Use Spawn Probability=false
    UseLoopCountLimit=false
   )
   /SetVariables(Emitter.VelDir:Vector3f=&RandomRangeVector(Minimum=<-0.1 0.5 0.7>, Maximum=<0.1 0.8 1>)) {isSetParameters=true}
  ]
  particleSpawnScript [
   /SetVariables(Particles.RandomFloat=&UniformRangedFloat(Minimum=0, Maximum=3)) {isSetParameters=true}
   &InitializeParticle(
    Color Maximum=<rgba 0.875 0.7711 0.6352 1>
    Color Minimum=<rgba 0.2604 0.2483 0.2257 1>
    Color Mode=::NewEnumerator2
    Lifetime Max=0.25
    Lifetime Min=0.15
    Lifetime Mode=::NewEnumerator1
    Sprite Size Max=<30 90>
    Sprite Size Min=<20 70>
    Sprite Size Mode=::NewEnumerator2
    UsePositionOffset=false
    Write Material Random=true
   ) {module=InitializeParticle001}
   &ShapeLocation(MaskSpawnGroup=false, Override Seed=false, Rotation Mode=::NewEnumerator1, Sphere Radius=5, UseOwnerScale=true)
   &AddVelocity(
    Cone Angle=165
    Cone Axis=$Emitter.VelDir
    Module.UseConeFalloff=false
    Module.UseVelDistribution=false
    Override Seed=false
    Velocity Mode=::NewEnumerator2
   ) {module=AddVelocity001}
   &AddVelocity(Module.Use Falloff Distance=false, Origin Offset=<0 5 0>, Override Seed=false, Velocity Mode=::NewEnumerator1) {enabled=false}
   &ApplyInitialForces()
   &SubUVAnimation(End Frame=3, SubUV Animation Mode:ENiagara_SubUVLookupMode=::NewEnumerator2)
  ]
  particleUpdateScript [
   &ParticleState() {module=UpdateAge}
   &SolveForcesAndVelocity(Clamp Velocity=false, Limit Acceleration=false)
   &ScaleSpriteSize(Scale Factor=&Vector2DFromCurve(
     Vector2Curve=DataInterface() {propertyValues {XCurve {keys [<at 0.0036 1.6897>,<at 0.9967 0.0582>]}, YCurve {keys [<at 0 1>,<at 1.002 3.7227>]}}}
     ScaleCurve=&Multiply_Vector2D_ByFloat(Float=0.65, Vector2D=<1 1>)
    ))
   &ScaleColor(
    Scale Alpha=&Multiply_Float(
     A=&FloatFromCurve(
      FloatCurve=DataInterface() {propertyValues {Curve {keys [{interpMode=RCIM_Cubic, tangentMode=RCTM_Break, time=0, value=1}, {interpMode=RCIM_Cubic, tangentMode=RCTM_Break, time=1, value=0}]}}}
      Scale Curve=&RandomRangeFloat(Minimum=0.1, Maximum=0.2)
     )
     B=&FNTimeOfDayScalar(Night Scale=1, Morning Scale=1, Evening Scale=1)
    )
    ScaleA=true
    ScaleRGB=false
   )
   &DynamicMaterialParameters(
    Index 0 Param 2=1
    Index 0 Param 3=1
    Index 0 Param 4=&FloatFromCurve(FloatCurve=DataInterface() {propertyValues {Curve {keys [<at 0.3 1>,<at 1 0.25>]}}})
    Module.Param0WriteEnabled=true
    Module.Param1WriteEnabled=true
    Module.Param2WriteEnabled=true
    Module.Param3WriteEnabled=true
   )
   &DynamicMaterialParameters(
    Index 0 Param 2=1
    Index 0 Param 3=1
    Index 0 Param 4=1
    Module.Param0WriteEnabled=true
    Module.Param1WriteEnabled=true
    Module.Param2WriteEnabled=true
    Module.Param3WriteEnabled=true
   ) {enabled=false, module=DynamicMaterialParameters001}
  ]
  rendererClasses [&NiagaraSpriteRendererProperties]
  renderers [{rendererClass=&NiagaraSpriteRendererProperties}]
 }
 FN_Core_Smoke_Ring {
  emitterUpdateScript [
   &EmitterState()
   /SetVariables(Emitter.VelDir:Vector3f=&RandomRangeVector(Minimum=<-0.1 0.5 0.7>, Maximum=<0.1 0.8 1>)) {isSetParameters=true}
   &SpawnBurst_Instantaneous(Spawn Count=&UniformRangedInt(Minimum=15, Maximum=20), Use Spawn Probability=false)
  ]
  particleSpawnScript [
   /SetVariables(Particles.VelCurveRand:NiagaraFloat=&FloatFromCurve(
     FloatCurve=DataInterface() {propertyValues {Curve {keys [
         {interpMode=RCIM_Cubic, leaveTangent=0.1132, tangentMode=RCTM_Break, time=0, value=0}
         {arriveTangent=1.9858, interpMode=RCIM_Cubic, leaveTangent=1.7203, tangentMode=RCTM_Break, time=0.3, value=0.4}
         {interpMode=RCIM_Cubic, tangentMode=RCTM_Break, time=1, value=1}
        ]}}}
     CurveIndex=&RandomRangeFloat(Minimum=0, Maximum=1)
    )) {isSetParameters=true}
   /SetVariables(Particles.InitSpriteRotRate=&RandomRangeFloat(Minimum=-1, Maximum=1)) {isSetParameters=true}
   &InitializeParticle(
    Color Maximum=<rgba 0.9375 0.9375 0.9375 0.4>
    Color Minimum=<rgba 0.6302 0.6302 0.6302 0.4>
    Color Mode=::NewEnumerator2
    Lifetime Max=0.6
    Lifetime Min=0.25
    Lifetime Mode=::NewEnumerator1
    Sprite Rotation Angle=&RandomRangeFloat(Minimum=0, Maximum=360)
    Sprite Rotation Mode=::NewEnumerator0
    Sprite Size Mode=::NewEnumerator1
    Uniform Sprite Size Max=100
    Uniform Sprite Size Min=80
    UsePositionOffset=false
    Write Material Random=true
   )
   &CameraOffset(Camera Offset Amount=&UniformRangedFloat(Minimum=10, Maximum=20))
   &ShapeLocation(MaskSpawnGroup=false, Override Seed=false, Rotation Mode=::NewEnumerator1, Sphere Radius=30, UseOwnerScale=true)
   &AddVelocity(
    Cone Angle=160
    Cone Axis=$Emitter.VelDir
    Module.UseConeFalloff=false
    Module.UseVelDistribution=false
    Override Seed=false
    Velocity Mode=::NewEnumerator2
    Velocity Speed=&Lerp_Float(A=80, B=500, Alpha=$Particles.VelCurveRand)
   )
  ]
  particleUpdateScript [
   &ParticleState()
   &Drag(Drag=&RandomRangeFloat(Minimum=7, Maximum=9), Use Linear Drag=true, Use Rotational Drag=false)
   &SpriteRotationRate(Rotation Rate=&Multiply_Float(
     A=$Particles.InitSpriteRotRate
     B=&FloatFromCurve(
      FloatCurve=DataInterface() {propertyValues {Curve {keys [
          {arriveTangent=-4.4624, interpMode=RCIM_Cubic, leaveTangent=-4.4624, tangentMode=RCTM_User, time=0, value=1}
          {arriveTangentWeight=0.1718, interpMode=RCIM_Cubic, tangentMode=RCTM_User, time=0.5, value=0.15}
         ]}}}
      Scale Curve=1500
     )
    ))
   &ScaleColor(
    Scale Alpha=&Multiply_Float(
     A=&RandomRangeFloat(Minimum=0.3, Maximum=0.4)
     B=&FNTimeOfDayScalar(Night Scale=1.3, Morning Scale=1, Evening Scale=1.3)
    )
    Scale RGB=&VectorFromFloat(Value=1)
    ScaleA=true
    ScaleRGB=false
   )
   &ScaleSpriteSize(Scale Factor=&Vector2DFromFloat(Value=&FloatFromCurve(
      FloatCurve=DataInterface() {propertyValues {Curve {keys [
          {arriveTangent=0.0418, interpMode=RCIM_Cubic, leaveTangent=0.0418, tangentMode=RCTM_User, time=0, value=0.5}
          {arriveTangent=0.8771, interpMode=RCIM_Cubic, leaveTangent=0.8771, tangentMode=RCTM_User, time=0.7, value=0.85}
          {arriveTangentWeight=0.3375, interpMode=RCIM_Cubic, tangentMode=RCTM_User, time=1, value=1}
         ]}}}
      Scale Curve=0.9
     )))
   &DynamicMaterialParameters(
    Index 0 Param 1=&Add_Float(
     A=&RandomRangeFloat(Minimum=0, Maximum=1)
     B=&FloatFromCurve(
      FloatCurve=DataInterface() {propertyValues {Curve {keys [
          {arriveTangent=2.7533, interpMode=RCIM_Cubic, leaveTangent=2.5749, tangentMode=RCTM_Break, time=0, value=0}
          {arriveTangent=0.0529, interpMode=RCIM_Cubic, leaveTangent=0.2021, tangentMode=RCTM_User, time=1, value=1}
         ]}}}
      Scale Curve=0.2
     )
    )
    Index 0 Param 3=&FloatFromCurve(FloatCurve=DataInterface() {propertyValues {Curve {keys [{arriveTangent=-53.9199, time=0, value=0}, {leaveTangent=0.2723, time=0.25, value=0}]}}})
    Index 0 Param 4=&FloatFromCurve(FloatCurve=DataInterface() {propertyValues {Curve {keys [<at 0 1>,<at 1 0>]}}})
    Index 1 Param 1=&FloatFromCurve(FloatCurve=DataInterface() {propertyValues {Curve {keys [<at 0 0>, {arriveTangent=4.2257, interpMode=RCIM_Cubic, leaveTangent=4.2257, tangentMode=RCTM_User, time=0.25, value=0}]}}})
    Index 1 Param 2=&FloatFromCurve(FloatCurve=DataInterface() {propertyValues {Curve {keys [<at 0 0>]}}})
    Index 1 Param 3=&FloatFromCurve(FloatCurve=DataInterface() {propertyValues {Curve {keys [{interpMode=RCIM_Cubic, leaveTangent=6.3917, tangentMode=RCTM_Break, time=0, value=0.1}]}}})
    Index 1 Param 4=&FloatFromCurve(FloatCurve=DataInterface() {propertyValues {Curve {keys [{arriveTangent=-0.136, leaveTangent=-0.136, time=0, value=1}]}}})
    Index 2 Param 1=&FloatFromCurve(
     FloatCurve=DataInterface() {propertyValues {Curve {keys [
         {arriveTangent=4.9626, interpMode=RCIM_Cubic, leaveTangent=2.5739, tangentMode=RCTM_User, time=0, value=0}
         {leaveTangent=-0.0719, time=1, value=1}
        ]}}}
     Scale Curve=3
    )
    Index 2 Param 2=&FloatFromCurve(FloatCurve=DataInterface() {propertyValues {Curve {keys [<at 0 0.5>,<at 1 1>]}}})
    Index 2 Param 3=&FloatFromCurve(
     FloatCurve=DataInterface() {propertyValues {Curve {keys [
         {interpMode=RCIM_Cubic, leaveTangent=-6.6324, leaveTangentWeight=0.1676, tangentMode=RCTM_Break, time=0, value=1}
         {interpMode=RCIM_Cubic, tangentMode=RCTM_User, time=0.2, value=0}
         {interpMode=RCIM_Cubic, leaveTangent=11.1233, tangentMode=RCTM_Break, time=1, value=1}
        ]}}}
     Scale Curve=5
    )
    Index 2 Param 4=&RandomRangeFloat(Minimum=0.9, Maximum=1.4)
    Module.Param0WriteEnabled=true
    Module.Param0WriteEnabled1=true
    Module.Param0WriteEnabled2=true
    Module.Param1WriteEnabled=false
    Module.Param1WriteEnabled1=true
    Module.Param1WriteEnabled2=true
    Module.Param2WriteEnabled=true
    Module.Param2WriteEnabled1=true
    Module.Param2WriteEnabled2=true
    Module.Param3WriteEnabled=true
    Module.Param3WriteEnabled1=true
    Module.Param3WriteEnabled2=true
    Write Parameter Index 1=true
    Write Parameter Index 2=true
   )
   &SolveForcesAndVelocity(Clamp Velocity=false, Limit Acceleration=false)
  ]
  rendererClasses [&NiagaraSpriteRendererProperties]
  renderers [
   {rendererClass=&NiagaraSpriteRendererProperties}
   {rendererClass=&NiagaraSpriteRendererProperties, rendererIndex=1}
  ]
 }
 FN_DebugGnomon {
  bEnabled=false
  emitterUpdateScript [
   &EmitterState(
    Life Cycle Mode=::NewEnumerator1
    Loop Behavior=::NewEnumerator1
    Loop Duration Mode=::NewEnumerator1
    Scalability Mode=::NewEnumerator1
    UseLoopDelay=false
   )
   &SpawnBurst_Instantaneous(Use Spawn Probability=false, UseLoopCountLimit=false)
  ]
  particleSpawnScript [&InitializeParticle(Color Mode=::NewEnumerator0, Mesh Scale Mode=::NewEnumerator3, Mesh Uniform Scale=5, UsePositionOffset=false, Write Material Random=true)]
  particleUpdateScript [&ParticleState()]
  rendererClasses [&NiagaraMeshRendererProperties]
  renderers [{rendererClass=&NiagaraMeshRendererProperties, rendererIndex=0}]
 }
 Guitar_Elements_Backup {
  emitterUpdateScript [
   &EmitterState()
   &SpawnBurst_Instantaneous(
    Spawn Count=&RandomRangeInt(Minimum=4, Maximum=6)
    Use Spawn Probability=false
    UseLoopCountLimit=false
   )
   &SpawnBurst_Instantaneous(
    Spawn Count=&RandomRangeInt(Minimum=15, Maximum=20)
    Spawn Group=1
    Use Spawn Probability=false
    UseLoopCountLimit=false
   ) {module=SpawnBurst_Instantaneous001}
   /SetVariables(Emitter.VelDir:Vector3f=&RandomRangeVector(Minimum=<-0.1 0.5 0.7>, Maximum=<0.1 0.8 1>)) {isSetParameters=true, module=SetVariables_#2}
  ]
  particleSpawnScript [
   /SetVariables(Particles.MeshIndex=&self:DetermineMeshProbability(
     Medium Mesh:NiagaraFloat=0.96
     Small Mesh:NiagaraFloat=0.5
     Randomness Mode:ENiagaraRandomnessMode=::NewEnumerator0
     Random Seed:NiagaraInt32=0
     Fixed Random Seed:NiagaraBool=false
    )) {isSetParameters=true, module=SetVariables_#3}
   &InitializeParticle(
    Color Maximum=<rgba 0.5395 0.4063 0.159 1>
    Color Minimum=<rgba 0.164 0.1125 0.0574 1>
    Color Mode=::NewEnumerator2
    Lifetime Mode=::NewEnumerator1
    Mass Mode=::NewEnumerator3
    Mesh Scale Mode=::NewEnumerator3
    Mesh Uniform Scale=&Multiply_Float(
     A=&Lerp_Float(
      A=0.5
      B=1.45
      Alpha=&FloatFromCurve(
       FloatCurve=DataInterface() {propertyValues {Curve {keys [<at 0 0 RCIM_Constant>, {arriveTangent=7.6187, interpMode=RCIM_Constant, leaveTangent=7.6187, time=0.65, value=1}]}}}
       CurveIndex=&RandomRangeFloat(Minimum=0, Maximum=1)
      )
     )
     B=4
    )
    UsePositionOffset=false
    Write Material Random=true
   )
   &InitialMeshOrientation(Use Rotation Vector=false)
   &ShapeLocation(MaskSpawnGroup=false, Override Seed=false, Sphere Radius=6, Sphere Surface Distribution=1, UseOwnerScale=true) {module=ShapeLocation001}
   &AddVelocity(
    Cone Angle=&MaskFloatBySpawnGroup(Masked Float=30, Passthrough Float=90)
    Cone Axis=$Emitter.VelDir
    Inner Cone Angle=10
    Module.UseConeFalloff=false
    Module.UseVelDistribution=false
    Override Seed=false
    Velocity Mode=::NewEnumerator2
    Velocity Speed=&RandomRangeFloat(Minimum=55, Maximum=666)
   )
   &ApplyInitialForces()
  ]
  particleUpdateScript [
   &ParticleState()
   &ScaleMeshSize(Scale Factor=&Multiply_VectorByFloat(
     Vector=&VectorFromFloat(Value=&FloatFromCurve(
       FloatCurve=DataInterface() {propertyValues {Curve {keys [
           <at 0 0 RCIM_Constant>
           <at 0.05 1 RCIM_Constant>
           {interpMode=RCIM_Cubic, leaveTangentWeight=0.3333, tangentMode=RCTM_User, time=0.75, value=1}
           {arriveTangentWeight=0.3333, interpMode=RCIM_Cubic, tangentMode=RCTM_User, time=1, value=0}
          ]}}}
       Scale Curve=&RandomRangeFloat(Minimum=0.25, Maximum=0.422)
      ))
     Float=&MaskFloatBySpawnGroup(
      Masked Float=&RandomRangeFloat(Minimum=1, Maximum=1.1)
      Passthrough Float=&RandomRangeFloat(Minimum=0.2, Maximum=0.5)
     )
    ))
   &GravityForce(Gravity=&Multiply_VectorByFloat(
     Vector=<0 0 -980>
     Float=&FloatFromCurve(FloatCurve=DataInterface() {propertyValues {Curve {keys [<at 0 0 RCIM_Constant>,<at 0.1 1>]}}}, CurveIndex=$Particles.Age)
    ))
   &Drag(
    Drag=&self:Gravity_Drag_Compensation(Drag Falloff:NiagaraFloat=2, Gravity Direction:Vector3f=<0 0 -1>)
    Use Linear Drag=true
    Use Rotational Drag=false
   )
   &UpdateMeshOrientation(
    Rotation Rate=&FloatFromCurve(
     FloatCurve=DataInterface() {propertyValues {Curve {keys [
         {tangentMode=RCTM_User, time=0, value=1}
         {arriveTangent=-0.0405, interpMode=RCIM_Cubic, leaveTangent=-0.0405, tangentMode=RCTM_User, time=0.25, value=1}
         {arriveTangent=-2.7856, interpMode=RCIM_Cubic, leaveTangent=-2.7856, tangentMode=RCTM_User, time=1, value=0.25}
        ]}}}
     Scale Curve=&RandomRangeFloat(Minimum=0.1, Maximum=2)
    )
    Rotation Vector=<1 1 1>
   )
   &Collision(
    Advanced Aging Rate=4
    Age Colliding Particles=true
    Analytical Collision Plane Position 1:Vector3f=<0 0 -16>
    CPU Collision Type=::NewEnumerator3
    Collision Enabled=&IsPlatformSetActive(New Platform Set=DataInterface() {propertyValues {Platforms {qualityLevelMask=-4}}})
    Enable Rest State=false
    Friction During a Bounce=0.05
    Max Number Of Collisions=0
    Module.EnableMaxCollisionCount=false
    Module.Randomize Collision Normal Vector=false
   )
   /Niagara/Modules/Collision/AlignParticlesWithCollisionPlane()
   &SolveForcesAndVelocity(Clamp Velocity=false, Limit Acceleration=false)
  ]
  rendererClasses [&NiagaraMeshRendererProperties]
  renderers [{rendererClass=&NiagaraMeshRendererProperties, rendererIndex=0}]
 }
 Impact_Spikes_Floor {
  emitterUpdateScript [
   &EmitterState(Life Cycle Mode=::NewEnumerator1, Loop Behavior=::NewEnumerator1, Loop Delay=0.03, Loop Duration=1, UseLoopDelay=false)
   &SpawnBurst_Instantaneous(Spawn Count=7, Use Spawn Probability=false, UseLoopCountLimit=false)
  ]
  particleSpawnScript [
   &InitializeParticle(
    Color Maximum=<rgba 0.875 0.7711 0.6352 1>
    Color Minimum=<rgba 0.2604 0.2483 0.2257 1>
    Color Mode=::NewEnumerator2
    Lifetime Max=0.25
    Lifetime Min=0.15
    Lifetime Mode=::NewEnumerator1
    Sprite Size Max=<20 50>
    Sprite Size Min=<10 25>
    Sprite Size Mode=::NewEnumerator2
    UsePositionOffset=false
    Write Material Random=true
   )
   &TorusLocation(
    Align Torus With Axis=false
    Coordinate Space:ENiagaraCoordinateSpace=::World
    Handle Radius=0
    Large Radius=0.16
    MaskSpawnGroup=false
    Radius Position=0
    Torus Distribution Mode:ENiagaraTorusDistributionMode=::NewEnumerator1
    U Position:NiagaraFloat=&RandomRangeFloat(Minimum=0, Maximum=1, Randomness Mode=::NewEnumerator2, Random Seed=?)
   )
   &AddVelocityFromPoint(
    Module.Use Falloff Distance=false
    Origin Offset=&RandomRangeVector(Minimum=<0 0.15 0>, Maximum=<0 0.15 -0.2>)
    Origin Offset Coordinate Space=::World
    Velocity Strength=&RandomRangeFloat(Minimum=32, Maximum=64)
   )
   &SubUVAnimation(End Frame=3, SubUV Animation Mode:ENiagara_SubUVLookupMode=::NewEnumerator2)
  ]
  particleUpdateScript [
   &ParticleState()
   &Drag(Drag=3, Use Linear Drag=true, Use Rotational Drag=false)
   &SolveForcesAndVelocity(Clamp Velocity=false, Limit Acceleration=false)
   &ScaleSpriteSize(
    Non-Uniform Curve Scale=&Multiply_Vector2D_ByFloat(Float=1.5, Vector2D=<1 1>)
    Non-Uniform Curve Sprite Scale=DataInterface() {propertyValues {
      XCurve {keys [{tangentMode=RCTM_User, time=0, value=1}, {tangentMode=RCTM_User, time=1, value=0.8}]}
      YCurve {keys [{tangentMode=RCTM_User, time=0, value=0.5}, {tangentMode=RCTM_User, time=1, value=1.7}]}
     }}
    Scale Sprite Size Mode=::NewEnumerator3
   )
   &ScaleColor(
    Scale Alpha=&Multiply_Float(
     A=&FloatFromCurve(
      FloatCurve=DataInterface() {propertyValues {Curve {keys [<at 0 0>,<at 0.1 1>,<at 0.2 0.8>,<at 1 0>]}}}
      Scale Curve=&FNTimeOfDayScalar(Night Scale=1, Morning Scale=1, Evening Scale=1)
     )
     B=0.1
    )
    ScaleA=true
    ScaleRGB=false
   )
   /SetVariables(Particles.Color2=&MultiplyLinearColor(Linear Color=$Particles.Color, Scale Factor=<1 1 1 0.2>)) {isSetParameters=true}
   &SpriteFacingAndAlignment(Facing Coordinate Space=::World, Sprite Facing=<0 0 1>, WriteAlignment=false, WriteFacing=true)
  ]
  rendererClasses [&NiagaraSpriteRendererProperties]
  renderers [
   {rendererClass=&NiagaraSpriteRendererProperties}
   {rendererClass=&NiagaraSpriteRendererProperties, rendererIndex=1}
  ]
 }
 Sparks {
  emitterUpdateScript [
   &EmitterState(
    Enable Distance Culling=true
    Max Distance Response=::KillImmediately
    Module.bApplySpawnCountScaleByDistance=true
    Module.bUseMaxDistance=true
    Module.bUseMinDistance=false
    Scalability Mode=::NewEnumerator1
    Scale Spawn Count=true
    Spawn Count Scale By Distance=DataInterface() {propertyValues {Curve {keys [<at 0.3676 0.9657>,<at 0.9962 0.4982>]}}}
   )
   &SpawnBurst_Instantaneous(Spawn Count=&UniformRangedInt(Minimum=1, Maximum=3), Use Spawn Probability=false)
  ]
  particleSpawnScript [
   &InitializeParticle(
    Color=<rgba 3 3 3 1>
    Lifetime Max=0.45
    Lifetime Min=0.166
    Lifetime Mode=::NewEnumerator1
    Sprite Size Mode=::NewEnumerator1
    Uniform Sprite Size Max=8
    Uniform Sprite Size Min=2.5
    UsePositionOffset=false
    Write Material Random=true
   )
   &ShapeLocation(MaskSpawnGroup=false, Override Seed=false, Sphere Radius=6, Sphere Surface Distribution=1, UseOwnerScale=true)
   &AddVelocity(
    Cone Angle=120
    Cone Axis=<0 0 1>
    Module.UseConeFalloff=false
    Module.UseVelDistribution=false
    Override Seed=false
    Velocity Mode=::NewEnumerator2
    Velocity Speed=&RandomRangeFloat(Minimum=222, Maximum=666)
   )
   &DynamicMaterialParameters(
    Index 0 Param 1=3
    Index 0 Param 2=20
    Index 0 Param 3=1
    Module.Param0WriteEnabled=true
    Module.Param1WriteEnabled=true
    Module.Param2WriteEnabled=true
    Module.Param3WriteEnabled=true
   )
  ]
  particleUpdateScript [
   &ParticleState()
   &GravityForce(Gravity=<0 0 -666>)
   &Drag(Drag=&UniformRangedFloat(Minimum=2, Maximum=3), Use Linear Drag=true, Use Rotational Drag=false)
   &SolveForcesAndVelocity(Clamp Velocity=false, Limit Acceleration=false)
   &ScaleSpriteSize(Scale Factor=&Vector2DFromCurve(
     Vector2Curve=DataInterface() {propertyValues {
       XCurve {keys [
         {arriveTangent=5.9728, interpMode=RCIM_Cubic, leaveTangent=5.9728, tangentMode=RCTM_User, time=0, value=0.1}
         {arriveTangent=-0.1, interpMode=RCIM_Cubic, leaveTangent=-0.1, tangentMode=RCTM_User, time=0.18, value=0.8}
         {interpMode=RCIM_Cubic, tangentMode=RCTM_User, time=1, value=0}
        ]}
       YCurve {keys [
         {arriveTangent=6.2715, interpMode=RCIM_Cubic, leaveTangent=6.2715, tangentMode=RCTM_User, time=0, value=0.1}
         {arriveTangent=-0.1001, interpMode=RCIM_Cubic, leaveTangent=-0.1001, tangentMode=RCTM_User, time=0.18, value=0.8}
         {interpMode=RCIM_Cubic, tangentMode=RCTM_User, time=0.9988, value=0}
        ]}
      }}
     ScaleCurve=<1 1>
    ))
   &ScaleSpriteSizeBySpeed(Max Scale Factor=<0.5 3>, Min Scale Factor=<0.5 1>, Velocity Threshold=666)
   &ScaleColor(
    Linear Color Curve=DataInterface() {propertyValues {
      AlphaCurve {keys [<at 0.333 1>,<at 1 0>]}
      BlueCurve {keys [<at 0 0.24>,<at 0.166 0.2>,<at 0.333 0>,<at 0.75 0>]}
      GreenCurve {keys [<at 0 3.12>,<at 0.166 2.6>,<at 0.333 0.6667>,<at 0.75 0.15>]}
      RedCurve {keys [<at 0 6>,<at 0.166 5>,<at 0.333 4>,<at 0.75 3>]}
     }}
    Scale Mode=::NewEnumerator2
   )
  ]
  rendererClasses [&NiagaraSpriteRendererProperties]
  renderers [{rendererClass=&NiagaraSpriteRendererProperties}]
 }
 Wood {
  emitterUpdateScript [
   &EmitterState()
   &SpawnBurst_Instantaneous(
    Spawn Count=&RandomRangeInt(Minimum=4, Maximum=8)
    Use Spawn Probability=false
    UseLoopCountLimit=false
   )
   &SpawnBurst_Instantaneous(
    Spawn Count=&RandomRangeInt(Minimum=13, Maximum=20)
    Spawn Group=1
    Use Spawn Probability=false
    UseLoopCountLimit=false
   ) {module=SpawnBurst_Instantaneous001}
   /SetVariables(Emitter.VelDir:Vector3f=&RandomRangeVector(Minimum=<-0.1 0.5 0.7>, Maximum=<0.1 0.8 1>)) {isSetParameters=true, module=SetVariables_#2}
  ]
  particleSpawnScript [
   /SetVariables(Particles.MeshIndex=&self:DetermineMeshProbability(
     Medium Mesh:NiagaraFloat=0.96
     Small Mesh:NiagaraFloat=0.5
     Randomness Mode:ENiagaraRandomnessMode=::NewEnumerator0
     Random Seed:NiagaraInt32=0
     Fixed Random Seed:NiagaraBool=false
    )) {isSetParameters=true, module=SetVariables_#3}
   &InitializeParticle(
    Color Maximum=<rgba 0.5395 0.4063 0.159 1>
    Color Minimum=<rgba 0.164 0.1125 0.0574 1>
    Color Mode=::NewEnumerator2
    Lifetime Mode=::NewEnumerator1
    Mass Mode=::NewEnumerator3
    Mesh Scale=&RandomRangeVector(
     Minimum=&SelectVectorFromArray(
      Direct Array Index=$Particles.MeshIndex
      Vector Selection Array=DataInterface() {propertyValues {FloatData [<1 1 1>, <2 2 2>, <0.2 0.2 0.2>]}}
     )
     Maximum=&SelectVectorFromArray(
      Direct Array Index=$Particles.MeshIndex
      Vector Selection Array=DataInterface() {propertyValues {FloatData [<1.5 1.5 1.5>, <3 3 3>, <0.3 0.3 0.3>]}}
     )
    )
    Mesh Scale Mode=::NewEnumerator4
    UsePositionOffset=false
    Write Material Random=true
   )
   &InitialMeshOrientation(Use Rotation Vector=false)
   &ShapeLocation(MaskSpawnGroup=false, Override Seed=false, Sphere Radius=6, Sphere Surface Distribution=1, UseOwnerScale=true) {module=ShapeLocation001}
   &AddVelocity(
    Cone Angle=&MaskFloatBySpawnGroup(Masked Float=30, Passthrough Float=90)
    Cone Axis=$Emitter.VelDir
    Inner Cone Angle=10
    Module.UseConeFalloff=false
    Module.UseVelDistribution=false
    Override Seed=false
    Velocity Mode=::NewEnumerator2
    Velocity Speed=&RandomRangeFloat(Minimum=55, Maximum=666)
   )
   &ApplyInitialForces()
  ]
  particleUpdateScript [
   &ParticleState()
   &ScaleMeshSize(Scale Factor=&Multiply_VectorByFloat(
     Vector=&VectorFromFloat(Value=&FloatFromCurve(
       FloatCurve=DataInterface() {propertyValues {Curve {keys [
           <at 0 0 RCIM_Constant>
           <at 0.05 1 RCIM_Constant>
           {interpMode=RCIM_Cubic, leaveTangentWeight=0.3333, tangentMode=RCTM_User, time=0.75, value=1}
           {arriveTangentWeight=0.3333, interpMode=RCIM_Cubic, tangentMode=RCTM_User, time=1, value=0}
          ]}}}
       Scale Curve=&RandomRangeFloat(Minimum=0.25, Maximum=0.422)
      ))
     Float=&MaskFloatBySpawnGroup(
      Masked Float=&RandomRangeFloat(Minimum=1.1, Maximum=1.2)
      Passthrough Float=&RandomRangeFloat(Minimum=0.3, Maximum=0.5)
     )
    ))
   &GravityForce(Gravity=&Multiply_VectorByFloat(
     Vector=<0 0 -980>
     Float=&FloatFromCurve(FloatCurve=DataInterface() {propertyValues {Curve {keys [<at 0 0 RCIM_Constant>,<at 0.1 1>]}}}, CurveIndex=$Particles.Age)
    ))
   &Drag(
    Drag=&self:Gravity_Drag_Compensation(Drag Falloff:NiagaraFloat=2, Gravity Direction:Vector3f=<0 0 -1>)
    Use Linear Drag=true
    Use Rotational Drag=false
   )
   &UpdateMeshOrientation(
    Rotation Rate=&FloatFromCurve(
     FloatCurve=DataInterface() {propertyValues {Curve {keys [
         {tangentMode=RCTM_User, time=0, value=1}
         {arriveTangent=-0.0405, interpMode=RCIM_Cubic, leaveTangent=-0.0405, tangentMode=RCTM_User, time=0.25, value=1}
         {arriveTangent=-2.7856, interpMode=RCIM_Cubic, leaveTangent=-2.7856, tangentMode=RCTM_User, time=1, value=0.25}
        ]}}}
     Scale Curve=&RandomRangeFloat(Minimum=0.1, Maximum=2)
    )
    Rotation Vector=<1 1 1>
   )
   &Collision(
    Advanced Aging Rate=4
    Age Colliding Particles=true
    Analytical Collision Plane Position 1:Vector3f=&AddVectorToPosition(Position=$Engine.Emitter.SimulationPosition, Vector=<0 0 -16>)
    CPU Collision Type=::NewEnumerator3
    Collision Enabled=&IsPlatformSetActive(New Platform Set=DataInterface() {propertyValues {Platforms {qualityLevelMask=-4}}})
    Enable Rest State=false
    Friction During a Bounce=0.05
    Max Number Of Collisions=0
    Module.EnableMaxCollisionCount=false
    Module.Randomize Collision Normal Vector=false
   )
   &AlignParticlesWithCollisionPlane()
   &SolveForcesAndVelocity(Clamp Velocity=false, Limit Acceleration=false)
  ]
  rendererClasses [&NiagaraMeshRendererProperties]
  renderers [{rendererClass=&NiagaraMeshRendererProperties, rendererIndex=0}]
 }
}
summary {
 systemName=NS_GroundImpact
 emitters [
  {emitterName=Sparks, rendererClasses [&NiagaraSpriteRendererProperties]}
  {emitterName=Impact_Spikes_Floor, rendererClasses [&NiagaraSpriteRendererProperties]}
  {emitterName=FN_Core_Smoke_Ring, rendererClasses [&NiagaraSpriteRendererProperties]}
  {emitterName=AirStreaks001, rendererClasses [&NiagaraSpriteRendererProperties]}
  {emitterName=Wood, rendererClasses [&NiagaraMeshRendererProperties]}
  {emitterName=Guitar_Elements_Backup, rendererClasses [&NiagaraMeshRendererProperties]}
  {bEnabled=false, emitterName=FN_DebugGnomon, rendererClasses [&NiagaraMeshRendererProperties]}
 ]
 userVariables [User.bIsCurie {type=NiagaraBool}]
}
